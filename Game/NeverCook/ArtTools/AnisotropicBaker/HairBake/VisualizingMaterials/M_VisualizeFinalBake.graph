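# Default Lit material in the anisotropic hair baker tools for visualizing the final bake, usable on skeletal meshes with mesh deformers.
# The Tangent texture's red and green channels are remapped to signed range, blue stays unsigned, and the vector is normalized.
# A SpiralBlur-Texture call also takes Tangent, with an adjustable blur distance.
# Scalar Aniso and Roughness parameters and a Normal map parameter complete the graph, which shows no output wiring.
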Material /Game/NeverCook/ArtTools/AnisotropicBaker/HairBake/VisualizingMaterials/M_VisualizeFinalBake
graph {
 properties {bUsedWithMeshDeformer=true, bUsedWithSkeletalMesh=true, pixelDepthOffsetMode=PDOM_Legacy, usedShadingModels=Default Lit}
 nodes [
  Constant() {r=1}
  MaterialFunctionCall!(
   TextureObject=TextureObjectParameter() {parameterName=Tangent}
   UV
   Distance=Divide(A=ScalarParameter() {defaultValue=5.9301, parameterName=d}) {constB=100}
   Distance Steps=Constant() {r=32}
   Radial Steps=Constant() {r=32}
   Radial Offset
   Kernel Power
  ) {materialFunction=/Engine/Functions/Engine_MaterialFunctions01/Texturing/SpiralBlur-Texture}
  Normalize(VectorInput=LinearInterpolate(
    A=Constant3Vector() {constant=<rgba -1 -1 0 0>}
    Alpha=TextureSampleParameter2D() {parameterName=Tangent, samplerType=SAMPLERTYPE_LinearColor, texture=/Game/Global/Textures/T_Default_Aniso}.RGB
   ))
  ScalarParameter() {defaultValue=0.8, parameterName=Aniso}
  ScalarParameter() {defaultValue=0.5, parameterName=Roughness}
  TextureSampleParameter2D() {
   parameterName=Normal
   samplerType=SAMPLERTYPE_Normal
   texture=/BRCosmetics/Characters/Player/Female/Medium/Bodies/F_MED_HighTower_Honeydew/Textures/T_F_MED_HighTower_Honeydew_FaceAcc_N
  }
 ]
}
