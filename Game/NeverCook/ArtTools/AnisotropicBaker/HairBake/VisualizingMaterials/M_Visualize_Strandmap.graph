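# Unlit debug material in the AnisotropicBaker hair-bake art tools, named for visualizing strand maps and stored under a NeverCook folder.
# Its only node is a texture parameter named Texture, defaulting to the T_Emporer_StrandMap test image, with no output connection shown.
# Usage flags enable it on skeletal meshes and mesh deformers.
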Material /Game/NeverCook/ArtTools/AnisotropicBaker/HairBake/VisualizingMaterials/M_Visualize_Strandmap
graph {
 properties {bUsedWithMeshDeformer=true, bUsedWithSkeletalMesh=true, pixelDepthOffsetMode=PDOM_Legacy, shadingModel=MSM_Unlit, usedShadingModels=Unlit}
 nodes [TextureSampleParameter2D() {parameterName=Texture, texture=/Game/NeverCook/ArtTools/AnisotropicBaker/TestImages/T_Emporer_StrandMap}]
}
