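# Debug material in the anisotropic hair bake tool for visualizing a captured tangent map, named for showing tangents in world space.
# It samples the Tangent texture parameter, defaulting to the HairAnisoCapture render target, and unpacks it with a lerp before normalizing.
# The vector feeds separate bias-scale and default-setting Transform nodes, alongside a sign-flip multiply on the unnormalized value.
# No connection to a material output is shown.
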Material /Game/NeverCook/ArtTools/AnisotropicBaker/HairBake/VisualizingMaterials/M_Visualize_TangentsAsWorld
graph {
 properties {bUsedWithMeshDeformer=true, bUsedWithSkeletalMesh=true, pixelDepthOffsetMode=PDOM_Legacy, usedShadingModels=Default Lit}
 nodes [
  n1 = TextureSampleParameter2D() {
   parameterName=Tangent
   samplerType=SAMPLERTYPE_LinearColor
   texture=/Game/NeverCook/ArtTools/AnisotropicBaker/HairBake/TransientRendertargets/HairAnisoCapture
  }
  n2 = LinearInterpolate(A=Constant3Vector() {constant=<rgba -1 -1 0 0>}, Alpha=@n1.RGB)
  n3 = Normalize(VectorInput=@n2)
  ConstantBiasScale(None=@n3)
  Multiply(A=@n2, B=VectorParameter() {defaultValue=<rgba -1 -1 1 1>}.RGB)
  Transform(None=@n3)
 ]
}
